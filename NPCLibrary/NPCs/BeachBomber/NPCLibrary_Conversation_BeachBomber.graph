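# Conversation database for the Beach Bomber NPC, who runs the SofDeez ice cream shop.
# The entry tag leads to one speech node with a default greeting, a first-ever-meeting line and joke lines for specific player skins, chosen by requirements and priority.
# The speech then branches through link nodes to the shared NPC library quests and basic services conversations.
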ConversationDatabase /NPCLibrary/NPCs/BeachBomber/NPCLibrary_Conversation_BeachBomber
entry_points [{tag=Conversation.Entry.BeachBomber, destinations [#1]}]
nodes [
 ConversationEntryPointNode_7 {
  id=#1
  class=ConversationEntryPointNode
  outputs [#2]
  properties {entryTag {tagName=Conversation.Entry.BeachBomber}}
 }
 Speech_C_4 {
  id=#2
  class=Speech_C
  outputs [#3, #4]
  properties {
   nodeName=""
   generalConfig {
    defaultMessage="Hey. We got vanilla, chocolate, and... raisin. I know. Supply chain issues."
    contextualMessages [
     {
      message="Welcome to SofDeez. You want ice cream, or... ice cream?"
      contextRequirements [{requirement {class=FortControllerRequirement_IsFirstEverConversationWithNPC}}]
      priority {curve {rowName=Tandem.Speech.Priority.FirstEverInteraction}}
     }
     {
      message="Sooo. Legally, SofDeez is mine now..."
      contextRequirements [{requirement {
         class=FortControllerRequirement_HasCID
         properties {allowedCIds [/BRCosmetics/Athena/Items/Cosmetics/Characters/CID_344_Athena_Commando_M_IceCream]}
        }}]
      priority {value=0, curve {curveTable=null, rowName=None}}
     }
     {
      message="How about a cone, on the house?"
      contextRequirements [{requirement {
         class=FortControllerRequirement_HasCID
         properties {allowedCIds [/BRCosmetics/Athena/Items/Cosmetics/Characters/CID_NPC_Athena_Commando_M_HenchmanSummer]}
        }}]
      priority {value=0, curve {curveTable=null, rowName=None}}
     }
     {
      message=Tell me about the beach. I miss it so much.
      requirementMatchPolicy=RequireAny
      contextRequirements [{requirement {
         class=FortControllerRequirement_HasCID
         properties {allowedCIds [
           /BRCosmetics/Athena/Items/Cosmetics/Characters/CID_A_129_Athena_Commando_M_CatBurglarSummer
           /BRCosmetics/Athena/Items/Cosmetics/Characters/CID_166_Athena_Commando_F_Lifeguard
           /BRCosmetics/Athena/Items/Cosmetics/Characters/CID_461_Athena_Commando_M_DriftSummer
           /BRCosmetics/Athena/Items/Cosmetics/Characters/CID_A_127_Athena_Commando_F_MechanicalEngineerSummer
           /BRCosmetics/Athena/Items/Cosmetics/Characters/CID_456_Athena_Commando_F_Sarong
           /BRCosmetics/Athena/Items/Cosmetics/Characters/CID_A_135_Athena_Commando_F_StreetFashionSummer
           /BRCosmetics/Athena/Items/Cosmetics/Characters/CID_A_130_Athena_Commando_M_HenchmanSummer
           &CID_814_Athena_Commando_M_BananaSummer
           /BRCosmetics/Athena/Items/Cosmetics/Characters/CID_812_Athena_Commando_F_RedRidingSummer
           /BRCosmetics/Athena/Items/Cosmetics/Characters/CID_A_153_Athena_Commando_F_BuffCatFan_TS2DR
           /BRCosmetics/Athena/Items/Cosmetics/Characters/CID_811_Athena_Commando_F_CandySummer
          ]}
        }}]
      priority {value=0, curve {curveTable=null, rowName=None}}
     }
     {
      message=Why couldn't you be the Herald of good summer sales?
      contextRequirements [{requirement {class=FortControllerRequirement_HasCID, properties {allowedCIds [&Character_TheHerald]}}}]
      priority {value=0, curve {curveTable=null, rowName=None}}
     }
     {
      message=I think ice cream is bad for cats. Sorry.
      requirementMatchPolicy=RequireAny
      contextRequirements [{requirement {
         class=FortControllerRequirement_HasCID
         properties {allowedCIds [
           &Character_ChillCat
           &CID_693_Athena_Commando_M_BuffCat
           /BRCosmetics/Athena/Items/Cosmetics/Characters/CID_776_Athena_Commando_M_ProfessorPup
          ]}
        }}]
      priority {value=0, curve {curveTable=null, rowName=None}}
     }
     {
      message=Don't get ice cream all over your... creepy metal goop hands.
      contextRequirements [{requirement {class=FortControllerRequirement_HasCID, properties {allowedCIds [&Character_Bites]}}}]
      priority {value=0, curve {curveTable=null, rowName=None}}
     }
     {
      message="I used to watch The Academy after school. You were my fave!"
      contextRequirements [{requirement {class=FortControllerRequirement_HasCID, properties {allowedCIds [&Character_PinkSpike]}}}]
      priority {value=0, curve {curveTable=null, rowName=None}}
     }
     {
      message=Maybe leave that suit outside. You might break something.
      contextRequirements [{requirement {class=FortControllerRequirement_HasCID, properties {allowedCIds [&Character_BadBear]}}}]
      priority {value=0, curve {curveTable=null, rowName=None}}
     }
     {
      message=That outfit looks way too warm for the beach.
      contextRequirements [{requirement {class=FortControllerRequirement_HasCID, properties {allowedCIds [&Character_RoseDust]}}}]
      priority {value=0, curve {curveTable=null, rowName=None}}
     }
    ]
   }
  }
 }
 ConversationLinkNode_0 {id=#4, class=ConversationLinkNode, properties {remoteEntryTag {tagName=Conversation.Entry.BasicServices}}}
 ConversationLinkNode_1 {id=#3, class=ConversationLinkNode, properties {remoteEntryTag {tagName=Conversation.Entry.NPCLibrary.Quests}}}
]
